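# Material function for the Creative Hologram Dancer device that produces a time-animated alpha controlling the lerp between hologram color curves.
# Time scaled by CurveLerpSpeed drives a sine, which is multiplied by CurveLerpHoldMultiplier and clamped to the default 0–1 range.
# The clamp flattens the wave, so the negative half holds at zero and multipliers above one lengthen the hold at one.
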MaterialFunction /Game/Creative/Devices/HologramDancer/Materials/MF_CP_Device_HologramDancer_ColorLerp
graph {
 properties {description=Controls Lerping of Hologram Color Curves}
 nodes [
  n1 = Sine(None=Multiply(A=Time(), B=ScalarParameter() {group=Color, parameterName=CurveLerpSpeed, sliderMax=1, sortPriority=61}))
  n2 = ScalarParameter() {defaultValue=1, group=Color, parameterName=CurveLerpHoldMultiplier, sortPriority=62}
  FunctionOutput(None=Clamp(None=Multiply(A=@n1, B=@n2) {constB=2})) {description=Multiply into Opacity or Emissive, outputName=Curve Lerp Alpha, sortPriority=1}
 ]
}
