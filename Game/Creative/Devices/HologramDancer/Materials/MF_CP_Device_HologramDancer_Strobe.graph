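# Material function that generates a time-driven strobe multiplier for the opacity or emissive of the Hologram Dancer device.
# A falling sawtooth from Time feeds two layers, a power-shaped flash and a Floor-quantized residual strobe, each contrast-adjusted and opacity-scaled before summing.
# The sum drives a lerp up from Min Opacity, which the Strobe parameter blends against 1 as an on/off control.
# A second output passes Strobe through as a lerp alpha for refraction.
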MaterialFunction /Game/Creative/Devices/HologramDancer/Materials/MF_CP_Device_HologramDancer_Strobe
graph {nodes [
  n1 = ScalarParameter() {
   defaultValue=0.25
   desc="Limits transparency. Above 0, it makes the hologram dim rather than fully vanish."
   group=Strobe
   parameterName=Min Opacity
  }
  n2 = Time()
  n3 = Multiply(A=ScalarParameter() {defaultValue=2, group=Strobe, parameterName=StrobeSpeed, sortPriority=2}) {constB=-1}
  n4 = ScalarParameter() {defaultValue=4, group=Strobe, parameterName=StrobePowerExp}
  n5 = Power(Base=Frac(None=Multiply(A=@n2, B=@n3)), Exp=@n4)
  n6 = ScalarParameter() {defaultValue=0.8, group=Strobe, parameterName=StrobeCheapContrast, sortPriority=36}
  n7 = ScalarParameter() {defaultValue=1, group=Strobe, parameterName=StrobeOpacity, sliderMax=1}
  n8 = Multiply(A=ScalarParameter() {defaultValue=12, group=Strobe, parameterName=ResidualStrobeStrobeSteps, sortPriority=50})
  n9 = Multiply(A=Frac(None=Multiply(A=@n2, B=@n3)), B=@n8)
  n10 = Power(Base=Divide(A=Floor(None=@n9), B=@n8), Exp=Divide(A=@n4))
  n11 = ScalarParameter() {defaultValue=0.2, group=Strobe, parameterName=ResidualStrobeCheapContrast, sortPriority=56}
  n12 = ScalarParameter() {defaultValue=0.5, group=Strobe, parameterName=ResidualStrobeOpacity, sliderMax=1, sortPriority=50}
  n13 = Add(
   A=Multiply(
    A=MaterialFunctionCall(In=@n5, Contrast=@n6) {materialFunction=&CheapContrast}.Result
    B=Clamp(None=@n7)
   )
   B=Multiply(
    A=MaterialFunctionCall(In=@n10, Contrast=@n11) {materialFunction=&CheapContrast}.Result
    B=Clamp(None=@n12)
   )
  )
  n14 = ScalarParameter() {desc=Lerp Switch On/Off, group=Strobe, parameterName=Strobe, sliderMax=1, sortPriority=1}
  FunctionOutput(None=LinearInterpolate(B=LinearInterpolate(A=@n1, Alpha=@n13), Alpha=@n14) {constA=1}) {description=Multiply into Opacity or Emissive, outputName=Strobe Opacity, sortPriority=1}
  FunctionOutput(None=@n14) {
   description=If there is Refraction: Use as Alpha input to Lerp between Refraction Value and 1
   outputName=Lerp Alpha for Refraction
   sortPriority=2
  }
 ]}
