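# Surface material for the Creative HologramDancer device: textured Default Lit base with an animated emissive glow, passed through MF_CreativeMoveTool.
# Glow color samples two curve-atlas gradients along UV1, blended by the ColorLerp function, with a silhouette color override.
# Panning Fmod scanlines, tiled differently on each UV half and center-masked, brighten the glow before gradient and panel-grid masks and Strobe opacity shape it.
# The Outer Mask and ScanlinePower branches are unconnected.
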Material /Game/Creative/Devices/HologramDancer/Materials/M_CP_Device_HologramDancer_01
graph {
 properties {bUseMaterialAttributes=true, pixelDepthOffsetMode=PDOM_Legacy, usedShadingModels=Default Lit}
 nodes [
  n1 = ScalarParameter() {group=Specular, parameterName=SpecRoughnessMin}
  n2 = ScalarParameter() {defaultValue=0.9, group=Specular, parameterName=SpecRoughnessMax}
  n3 = TextureSampleParameter2D() {parameterName=SpecularMasks, texture=/Game/Creative/Devices/HologramDancer/Textures/T_CP_Device_HologramDancer_01_S}
  n4 = TextureSampleParameter2D() {
   parameterName=Normals
   samplerType=SAMPLERTYPE_Normal
   texture=/Game/Creative/Devices/HologramDancer/Textures/T_CP_Device_HologramDancer_01_N
  }
  n5 = Constant()
  n6 = MaterialFunctionCall(UV Channel=TextureCoordinate() {coordinateIndex=1}) {materialFunction=&LinearGradient}
  n7 = ScalarParameter() {defaultValue=0.5, group=Glow, parameterName=ScanLine Center Mask Exponent, sortPriority=61}
  n8 = TextureCoordinate() {coordinateIndex=1}
  n9 = MaterialFunctionCall(
   In=OneMinus(None=MaterialFunctionCall(UV Channel=@n8) {materialFunction=&LinearGradient}.UGradient)
   Contrast=Constant() {r=100}
  ) {desc=Half Mask, materialFunction=&CheapContrast}
  n10 = OneMinus(None=@n9.Result)
  n11 = Multiply(
   A=MaterialFunctionCall(
    In=Power(Base=@n6.VGradient, Exp=@n7)
    Contrast=ScalarParameter() {defaultValue=0.5, group=Glow, parameterName=ScanLine Center Mask Contrast, sortPriority=62}
   ) {materialFunction=&CheapContrast}.Result
   B=@n10
  )
  n12 = ScalarParameter() {defaultValue=10, group=Glow, parameterName=ScanlineTile, sortPriority=42}
  n13 = Divide(B=@n12)
  n14 = Time()
  n15 = ScalarParameter() {defaultValue=0.1, group=Glow, parameterName=ScanlinePanSpeed, sortPriority=41}
  n16 = MaterialFunctionCall(UV Channel=Add(A=TextureCoordinate() {coordinateIndex=1}, B=Multiply(A=@n14, B=@n15))) {materialFunction=&LinearGradient}
  n17 = Divide(B=Divide(A=@n12))
  n18 = LinearInterpolate(A=Divide(A=Fmod(A=@n16.VGradient, B=@n13), B=@n13), B=Divide(A=Fmod(A=@n16.VGradient, B=@n17), B=@n17), Alpha=@n9.Result)
  n19 = Multiply(A=@n18, B=OneMinus(None=@n18)) {desc=Reverse and Combine Gradient}
  n20 = MaterialFunctionCall(In=@n19, Contrast=ScalarParameter() {defaultValue=0.15, group=Glow, parameterName=ScanlineContrast, sortPriority=46}) {materialFunction=&CheapContrast}
  n21 = ScalarParameter() {defaultValue=1, group=Glow, parameterName=ScanLine Center Mask Intensity, sliderMax=1, sortPriority=60}
  n22 = Multiply(
   A=LinearInterpolate(A=@n20.Result, B=LinearInterpolate(A=@n20.Result, B=@n5, Alpha=@n11), Alpha=Clamp(None=@n21))
   B=ScalarParameter() {defaultValue=10, group=Glow, parameterName=ScanlineIntensity, sortPriority=40}
  )
  n23 = TextureCoordinate() {coordinateIndex=1}
  n24 = Power(
   Base=MaterialFunctionCall(UV Channel=@n23) {desc=Used as Gradient Map Coordinates, materialFunction=&LinearGradient}.VGradient
   Exp=ScalarParameter() {defaultValue=0.4, group=Glow, parameterName=Curve Power, sortPriority=20}
  )
  n25 = CurveAtlasRowParameter(CurveTime=@n24) {
   atlas=/Game/Creative/Devices/HologramDancer/LinearColorCurves/CurveAtlas_CP_Device_HologramDancer_01
   curve=/Game/Creative/Devices/HologramDancer/LinearColorCurves/Curve_CP_Device_HologramDancer_01
   desc=Curve Atlas
   group=Glow
   parameterName=Curve01
   sortPriority=20
  }
  n26 = CurveAtlasRowParameter(CurveTime=@n24) {
   atlas=&CurveAtlas_CP_Device_HologramDancer_01
   curve=/Game/Creative/Devices/HologramDancer/LinearColorCurves/Curve_CP_Device_HologramDancer_02
   defaultValue=1
   desc=Curve Atlas
   group=Glow
   parameterName=Curve02
   sortPriority=20
  }
  n27 = MaterialFunctionCall() {materialFunction=&/MF_CP_Device_HologramDancer_ColorLerp}
  n28 = LinearInterpolate(
   A=Multiply(
    A=ScalarParameter() {defaultValue=10, group=Glow, parameterName=Glow Intensity Multiplier, sortPriority=10}
    B=LinearInterpolate(A=@n25, B=@n26, Alpha=@n27.Curve Lerp Alpha)
   )
   B=VectorParameter() {group=Silhouette, parameterName=Silhouette Color, sortPriority=2, defaultValue=<rgba 0 0 0 1>}.RGB
   Alpha=ScalarParameter() {group=Silhouette, parameterName=Silhouette Color Override, sortPriority=1}
  )
  n29 = OneMinus(None=MaterialFunctionCall(Float2=@n23) {materialFunction=&BreakOutFloat2Components}.G)
  n30 = ScalarParameter() {defaultValue=0.1, group=Glow, parameterName=Glow UV Mask Exponent, sortPriority=10}
  n31 = LinearInterpolate(
   A=Constant()
   B=Add(A=Multiply(A=@n22, B=@n28), B=@n28)
   Alpha=MaterialFunctionCall(In=Power(Base=@n29, Exp=@n30), Contrast=ScalarParameter() {defaultValue=1, group=Glow, parameterName=Glow UV Mask Contrast, sortPriority=11}) {materialFunction=&CheapContrast}.Result
  ) {constAlpha=1}
  n32 = Clamp(None=Constant() {desc=Panel Grid Glow Mask, r=0.8})
  n33 = LinearInterpolate(A=@n31, B=Multiply(A=@n31, B=LinearInterpolate(B=@n3.R, Alpha=@n32) {constA=1}), Alpha=@n10) {constAlpha=1}
  n34 = MaterialFunctionCall() {materialFunction=&/MF_CP_Device_HologramDancer_Strobe}
  MaterialFunctionCall(MA In=MakeMaterialAttributes!(
    BaseColor=TextureSampleParameter2D() {parameterName=Diffuse, texture=/Game/Creative/Devices/HologramDancer/Textures/T_CP_Device_HologramDancer_01_D}.RGB
    Metallic=@n3.G
    Specular=@n3.R
    Roughness=LinearInterpolate(A=@n1, B=@n2, Alpha=@n3.B)
    Anisotropy
    EmissiveColor=Multiply(A=@n33, B=@n34.Strobe Opacity)
    Opacity
    OpacityMask
    Normal=@n4.RGB
    Tangent
    WorldPositionOffset
    SubsurfaceColor
    ClearCoat
    ClearCoatRoughness
    AmbientOcclusion
    Refraction
    CustomizedUVs_0
    CustomizedUVs_1
    CustomizedUVs_2
    CustomizedUVs_3
    CustomizedUVs_4
    CustomizedUVs_5
    CustomizedUVs_6
    CustomizedUVs_7
    PixelDepthOffset
    ShadingModel
    Displacement
   )) {materialFunction=&MF_CreativeMoveTool}
  n35 = Power(
   Base=MaterialFunctionCall(UV Channel=TextureCoordinate() {coordinateIndex=1}) {materialFunction=&LinearGradient}.VGradient
   Exp=ScalarParameter() {defaultValue=2, group=Glow, parameterName=Outer Mask Exponent}
  )
  n36 = ScalarParameter() {defaultValue=1, group=Glow, parameterName=Outer Mask Intensity, sliderMin=1}
  Multiply(B=LinearInterpolate(B=LinearInterpolate(B=@n35, Alpha=@n9.Result) {constA=1}, Alpha=Clamp(None=@n36)) {constA=1})
  OneMinus(None=@n6.VGradient)
  Power(Base=@n19, Exp=ScalarParameter() {defaultValue=1, group=Glow, parameterName=ScanlinePower, sortPriority=45})
 ]
}
