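# Translucent unlit hologram material for the NPC dancer of the Creative HologramDancer device, with a flat-color silhouette switch.
# World-aligned noise, flickered by position-seeded sines, drives vertex jitter and joins fresnel, actor-anchored screen scanlines, height gradient and TurnOn pulse in shaping opacity.
# That mask indexes two curve-atlas colors for emissive, plus reflection, sun specular and rim glow.
# A custom-depth compare culls hidden layers; strobe function modulates opacity and refraction; optional FaceCamera rotates vertices.
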Material /Game/Creative/Devices/HologramDancer/Materials/M_CP_Device_HologramDancer_NPC_01
graph {
 properties {
  allowTranslucentCustomDepthWrites=true
  bAutomaticallySetUsageInEditor=false
  bEnableResponsiveAA=true
  bUseMaterialAttributes=true
  bUsedWithClothing=true
  bUsedWithMeshDeformer=true
  bUsedWithMorphTargets=true
  bUsedWithSkeletalMesh=true
  blendMode=BLEND_Translucent
  numCustomizedUVs=4
  opacityMaskClipValue=0
  pixelDepthOffsetMode=PDOM_Legacy
  refractionMethod=RM_IndexOfRefraction
  shadingModel=MSM_Unlit
  usedShadingModels=Unlit
 }
 nodes [
  Constant2Vector() {g=-1, r=1}
  n1 = Divide(A=ObjectRadius(), B=ScalarParameter() {defaultValue=0.1, group=Noise, parameterName=NoiseSize})
  n2 = Multiply(A=Time(), B=VectorParameter() {group=Noise, parameterName=NoisePanSpeed, defaultValue=<rgba -0.08 0.07 -0.2 0>}.RGB)
  n3 = MaterialFunctionCall(
   TextureObject=TextureObject() {samplerType=SAMPLERTYPE_LinearColor, texture=&T_NoisePack_LinearColor}
   TextureSize=Multiply(A=VectorParameter() {group=Noise, parameterName=NoiseTileMult, defaultValue=<rgba 1 1 0.1 1>}.RGB, B=@n1)
   WorldPosition=Add(A=WorldPosition().XYZ, B=Multiply(A=@n1, B=@n2))
  ) {materialFunction=&WorldAlignedTexture}
  n4 = ActorPositionWS()
  n5 = MaterialFunctionCall(Float2=Round(None=Divide(A=@n4) {constB=2000})) {materialFunction=&BreakOutFloat2Components}
  n6 = Add(A=Time(), B=Add(A=@n5.R, B=@n5.R))
  n7 = Multiply(A=ScalarParameter() {defaultValue=0.5, parameterName=FlickerSpeed}, B=@n6)
  n8 = AppendVector(A=Sine(None=@n7) {period=1.6}, B=Sine(None=@n7) {period=1.42})
  n9 = Sine(None=@n7) {period=1.51}
  n10 = Sine(None=@n7) {period=2.1}
  n11 = ScalarParameter() {defaultValue=0.5, group=Noise, parameterName=NoiseFrequency}
  n12 = Multiply(B=@n11) {constA=0.25}
  n13 = Add(A=Sine(None=@n7) {period=0.72}, B=Multiply(B=@n11) {constA=0.75}) {constB=0.75}
  n14 = Add(A=Multiply(B=@n11) {constA=0.9, constB=0.9}, B=Sine(None=@n7) {period=1.25}) {constB=0.9}
  n15 = Multiply(A=Multiply(A=Add(A=@n10, B=@n12), B=@n13), B=Multiply(A=@n13, B=Ceil(None=@n14)))
  n16 = ConstantBiasScale(None=Sine(None=@n6) {period=0.075})
  n17 = Multiply(A=Ceil(None=AppendVector(A=@n8, B=@n9)), B=Multiply(A=Saturate(None=@n15), B=LinearInterpolate(Alpha=@n16) {constA=0.7}))
  n18 = ScalarParameter() {defaultValue=2, group=Noise, parameterName=NoiseMult}
  n19 = Multiply(
   A=Multiply(A=Power(Base=@n3.XY Texture) {constExponent=4}, B=ScalarParameter() {defaultValue=2, group=Noise, parameterName=NoiseMaskMult}) {constB=20}
   B=Add(A=ScalarParameter() {group=Noise, parameterName=NoiseAdd}, B=Multiply(A=@n17, B=@n18))
  )
  n20 = WorldPosition()
  n21 = ComponentMask(None=MaterialFunctionCall() {materialFunction=&ObjectPivotPoint}.Object Pivot Location) {b=true}
  n22 = Add(A=@n21, B=ScalarParameter() {defaultValue=300, parameterName=PulseHeight}) {constB=1600}
  n23 = ScalarParameter() {defaultValue=1, parameterName=TurnOn}
  n24 = ScalarParameter() {defaultValue=33, parameterName=PulseRadius}
  n25 = Multiply(A=@n24) {constB=0.1}
  n26 = Multiply(
   A=SphereMask(
    A=ComponentMask(None=@n20.XYZ) {b=true}
    B=LinearInterpolate(A=@n21, B=@n22, Alpha=@n23) {constA=-0.25, constB=1.2}
    Radius=LinearInterpolate(A=@n24, B=@n25, Alpha=@n23) {constA=500, constB=100}
   ) {hardnessPercent=0}
   B=OneMinus(None=@n23)
  )
  n27 = Add(A=Desaturation(None=Saturate(None=@n19)), B=Multiply(A=@n26) {constB=2})
  n28 = QualitySwitch(
   Default=MaterialFunctionCall(In=@n27, Contrast=ScalarParameter() {defaultValue=0.2, group=Noise, parameterName=NoiseContrast}) {materialFunction=&CheapContrast}.Result
   Low=Constant()
  )
  n29 = Saturate(None=@n28)
  n30 = Multiply(A=Constant() {r=60}, B=LinearInterpolate(Alpha=Ceil(None=@n29)) {constA=1, constB=1.25})
  n31 = ConstantBiasScale(None=ScreenPosition().ViewportUV) {bias=-0.5, scale=2}
  n32 = MaterialFunctionCall(Float2=ViewSize()) {materialFunction=&BreakOutFloat2Components}
  n33 = AppendVector(A=Constant() {r=1}, B=Divide(A=@n32.G, B=@n32.G))
  n34 = Multiply(
   A=Constant2Vector() {g=-1, r=1}
   B=MaterialFunctionCall(World Position=ActorPositionWS() {desc=Also use Object Position}) {materialFunction=&TransformToClipSpace}.Clip Space XY
  )
  n35 = Subtract(
   A=ConstantBiasScale(None=Multiply(A=@n31, B=@n33) {desc=Keep 1:1 aspect ratio regardless of screen aspect ratio}) {scale=2}
   B=Multiply(A=Multiply(A=@n33, B=@n34)) {constB=0.5, desc=Halve offsets to account for 0...1 UV space}
  )
  n36 = Add(
   A=MaterialFunctionCall(UVs=@n35, "Rotation Angle (0-1)"=Constant() {r=0.25}) {materialFunction=&CustomRotator}.Rotated Values
   B=Multiply(A=Time(), B=Constant() {r=-0.02})
  )
  n37 = ConstantBiasScale(None=Sine(None=Multiply(A=@n30, B=@n36) {constA=1}))
  n38 = DotProduct(A=VertexNormalWS(), B=Constant3Vector() {constant=<rgba 0 0 1 0>})
  LinearInterpolate(A=@n37, Alpha=Power(Base=Abs(None=@n38)) {constExponent=5}) {constB=0.25}
  MaterialFunctionCall!(Material) {materialFunction=/Game/Packages/Fortress_SharedMaterials/Base_Material_Functions/BaseCharacterShaderFunctions/BaseFN_CharacterOpacity}
  n39 = Constant3Vector() {constant=<rgba 0 0 1 0>}
  n40 = CameraPositionWS()
  n41 = ActorPositionWS()
  n42 = Normalize(VectorInput=Multiply(A=Constant3Vector() {constant=<rgba 1 1 0 0>}, B=Subtract(A=@n40, B=@n41)))
  n43 = Normalize(VectorInput=Constant3Vector() {constant=<rgba 0 1 0 0>})
  n44 = DotProduct(A=@n42, B=Normalize(VectorInput=Constant3Vector() {constant=<rgba 1 0 0 0>}))
  n45 = Multiply(
   A=Divide(A=MaterialFunctionCall!(Normalized V1=@n42, Normalized V2=@n43) {materialFunction=/Engine/Functions/Engine_MaterialFunctions02/Math/AngleBetweenVectors}.Angle In Degrees) {constB=360}
   B=Multiply(A=Sign(None=@n44)) {constB=-1}
  )
  n46 = Normalize(VectorInput=Subtract(A=WorldPosition().XYZ, B=ObjectPositionWS()))
  n47 = Constant3Vector() {constant=<rgba 1 1 0 0>}
  n48 = ScalarParameter() {defaultValue=10, group=Noise, parameterName=WPOStrength}
  n49 = Multiply(A=ObjectRadius(), B=ScalarParameter() {defaultValue=0.025, group=Noise, parameterName=WPOStrengthFromObjectRadius}) {constB=0.05}
  n50 = Multiply(A=@n27, B=Multiply(A=Multiply(A=@n46, B=@n47), B=Add(A=@n48, B=@n49))) {constB=0}
  n51 = Add(A=@n50, B=RotateAboutAxis(NormalizedRotationAxis=@n39, RotationAngle=@n45, PivotPoint=@n41, Position=WorldPosition().XYZ))
  n52 = Add(
   A=StaticSwitchParameter(True=@n51, False=@n50) {parameterName=FaceCamera}
   B=MaterialFunctionCall(Offset Amount=ScalarParameter() {group=WPO, parameterName=CameraOffset}) {materialFunction=/Engine/Functions/Engine_MaterialFunctions02/WorldPositionOffset/CameraOffset}.Camera Offset
  )
  NamedRerouteDeclaration(None=StaticSwitchParameter(
    True=MaterialFunctionCall!(WPO=@n52) {materialFunction=/Game/Packages/Fortress_SharedMaterials/Base_Material_Functions/Fortnite_HideElementsWithVertexColor}.Result
    False=@n52
   ) {group=Opacity Mask, parameterName=Use Vertex Colors for Mask}) {name=WPO, nodeColor=<rgba 1 0.8203 0 1>}
  n53 = MaterialFunctionCall() {materialFunction=&/MF_CP_Device_HologramDancer_Strobe}
  n54 = Constant() {r=1}
  n55 = Constant()
  n56 = Saturate(None=Add(A=@n53.Lerp Alpha for Refraction, B=QualitySwitch(Default=@n54, Low=@n54, High=@n54, Medium=@n54, Epic=@n55)))
  NamedRerouteDeclaration(None=LinearInterpolate(
    A=Constant() {r=0.95}
    Alpha=StaticSwitchParameter(True=@n56, False=@n53.Lerp Alpha for Refraction) {desc=Silhouette Switch, group=Silhouette, parameterName=IsSilhouette, sortPriority=0}
   ) {constA=1}) {name=Refraction, nodeColor=<rgba 1 0.8203 0 1>}
  n57 = ScalarParameter() {defaultValue=1.5, parameterName=OpacityMult}
  n58 = ComponentMask(None=MaterialFunctionCall() {materialFunction=&ObjectPivotPoint}.Object Pivot Location) {b=true}
  n59 = Multiply(
   A=ScalarParameter() {defaultValue=2.75, desc="Needs to match character scale to ", group=Color, parameterName=ObjectScale, sortPriority=81}
   B=ScalarParameter() {defaultValue=180, group=Color, parameterName=ObjectHeight, sortPriority=81}
  ) {constB=180}
  n60 = Multiply(A=@n59) {constB=1.25}
  n61 = SphereMask(A=ComponentMask(None=WorldPosition().XYZ) {b=true}, B=Add(A=@n58, B=@n59) {constB=1600}, Radius=@n60) {hardnessPercent=0}
  n62 = WorldPosition()
  n63 = SphereMask(A=Multiply(A=ComponentMask(None=@n62.XYZ) {b=true}) {constB=-1}, B=Add(A=Multiply(A=@n58) {constB=-1}, B=@n59) {constB=1600}, Radius=@n60) {hardnessPercent=0}
  n64 = ScalarParameter() {defaultValue=1, parameterName=FresnelExp}
  n65 = LinearInterpolate(B=ScalarParameter() {defaultValue=0.6, group=Noise, parameterName=NoiseFresnelXPMult}, Alpha=@n28) {constA=1, constB=0.5}
  n66 = StaticSwitchParameter(
   True=MaterialFunctionCall(Rotation Axis=@n39, Rotation Angle=@n45) {materialFunction=&FixRotateAboutAxisNormals}.New World Space Vertex Normal
   False=PixelNormalWS()
  ) {parameterName=FaceCamera}
  n67 = Fresnel(ExponentIn=Multiply(A=@n65, B=ScalarParameter() {defaultValue=1, parameterName=EdgeFresnelXP}), Normal=@n66) {baseReflectFraction=0, exponent=1}
  n68 = Multiply(
   A=Fresnel(ExponentIn=Multiply(A=@n64, B=@n65), BaseReflectFractionIn=ScalarParameter() {defaultValue=0.5, parameterName=FresnelMin}, Normal=@n66) {baseReflectFraction=0, exponent=0.5}
   B=Saturate(None=LinearInterpolate(Alpha=@n67) {constA=5, constB=-0.4})
  )
  n69 = ScalarParameter() {defaultValue=-0.4, group=Noise, parameterName=NoiseBrightness}
  n70 = Constant() {r=0.5}
  n71 = Multiply(A=Add(A=Power(Base=@n68), B=Multiply(A=@n69, B=@n29) {constB=0.1}), B=Saturate(None=Add(A=@n70, B=@n37)))
  n72 = ScalarParameter() {defaultValue=1, group=Color, parameterName=VerticalGradientInfluence, sortPriority=80}
  n73 = Add(A=LinearInterpolate(A=@n71, B=Add(A=@n61, B=@n63), Alpha=@n72), B=@n26) {constA=1}
  n74 = Constant() {r=0.6}
  n75 = Multiply(A=@n28) {constB=1.5}
  n76 = LinearInterpolate(
   A=Clamp(None=Subtract(A=@n74, B=@n75)) {minDefault=-0.25}
   B=Add(A=@n75)
   Alpha=MaterialFunctionCall(Float2=Power(Base=@n37) {constExponent=1.5}) {materialFunction=&BreakOutFloat2Components}.R
  ) {constA=0.5}
  n77 = Multiply(A=Saturate(None=Multiply(A=@n73, B=@n76)), B=@n57)
  n78 = PixelDepth()
  n79 = Multiply(
   A=ScalarParameter() {defaultValue=-1, parameterName=BiasSorting_Far}
   B=MaterialFunctionCall() {materialFunction=&ObjectScale}.Scale Z
  )
  n80 = Constant()
  n81 = If(
   A=ComponentMask(None=Add(A=@n78.R, B=@n79)) {r=true}
   B=ComponentMask(None=SceneTexture() {sceneTextureId=PPI_CustomDepth}.Color) {r=true}
   "A > B"=@n80
   "A == B"=@n80
   "A < B"=Constant() {r=1}
  )
  n82 = Multiply(
   A=LinearInterpolate(B=ParticleColor().A, Alpha=ScalarParameter() {parameterName=UseParticleColor}) {constA=1, desc=Workaround for skel mesh bug}
   B=StaticSwitchParameter(True=@n57, False=Multiply(A=@n77, B=@n81)) {desc=Silhouette Switch, group=Silhouette, parameterName=IsSilhouette, sortPriority=0}
  )
  n83 = ScalarParameter() {defaultValue=200, parameterName=FadeDistance}
  NamedRerouteDeclaration(None=Multiply(A=DepthFade(Opacity=@n82, FadeDistance=@n83), B=@n53.Strobe Opacity) {constA=1}) {name=Opacity, nodeColor=<rgba 1 0.8203 0 1>}
  n84 = ComponentMask(None=Transform(None=ReflectionVectorWS()) {transformSourceType=TRANSFORMSOURCE_World, transformType=TRANSFORM_View}) {g=true, r=true}
  n85 = ScalarParameter() {defaultValue=1, group=Reflection, parameterName=ReflectionMapTilingU, sortPriority=5}
  n86 = ScalarParameter() {defaultValue=1, group=Reflection, parameterName=ReflectionMapTilingV, sortPriority=5}
  n87 = TextureSampleParameter2D(UVs=Multiply(A=ConstantBiasScale(None=@n84), B=AppendVector(A=@n85, B=@n86))) {group=Reflection, parameterName=SphereReflectionMap, sortPriority=1, texture=/Game/Packages/Fortress_SharedMaterials/Textures/ReflectionMap2}
  n88 = Multiply(
   A=Power(Base=@n87.RGB, Exp=ScalarParameter() {defaultValue=4, group=Reflection, parameterName=ReflectionExponent})
   B=VectorParameter() {group=Reflection, parameterName=Reflection Tint, sortPriority=2, defaultValue=<rgba 1 1 1 1>}.RGB
  )
  n89 = ScalarParameter() {defaultValue=1, group=Reflection, parameterName=Reflection Multiplier}
  n90 = CollectionParameter() {
   collection=&FortniteMaterialParameters
   parameterName=SunAndMoonModelDirectionalVector
  }
  n91 = DotProduct(A=ComponentMask(None=@n90) {b=true, g=true, r=true}, B=ReflectionVectorWS())
  n92 = VectorParameter() {parameterName=LightDirection, defaultValue=<rgba 0 0 1 1>}
  n93 = LinearInterpolate(
   A=VertexNormalWS()
   B=Transform(None=TextureSampleParameter2D() {group=Texture, parameterName=Normal, samplerType=SAMPLERTYPE_Normal, texture=/Engine/EngineMaterials/BaseFlattenNormalMap}.RGB)
   Alpha=ScalarParameter() {group=Lighting, parameterName=UseNormal}
  )
  n94 = Multiply(A=Clamp(None=DotProduct(A=@n92.RGB, B=@n93)), B=ScalarParameter() {defaultValue=0.2, group=Lighting, parameterName=NormalBrightnessAdd}) {constB=0.15}
  n95 = MaterialFunctionCall(Float2=Saturate(None=Add(A=@n94, B=@n73))) {materialFunction=&BreakOutFloat2Components}
  n96 = CurveAtlasRowParameter(CurveTime=@n95.R) {
   atlas=/Game/Creative/Devices/HologramDancer/LinearColorCurves/CurveAtlas_CP_Device_HologramDancer_01
   curve=/Game/Creative/Devices/HologramDancer/LinearColorCurves/Curve_CP_Device_HologramDancer_01
   desc=Curve Atlas
   group=Color
   parameterName=Curve01
  }
  n97 = CurveAtlasRowParameter(CurveTime=@n95.R) {
   atlas=&CurveAtlas_CP_Device_HologramDancer_01
   curve=/Game/Creative/Devices/HologramDancer/LinearColorCurves/Curve_CP_Device_HologramDancer_02
   defaultValue=1
   desc=Curve Atlas
   group=Color
   parameterName=Curve02
  }
  n98 = LinearInterpolate(
   A=@n96
   B=@n97
   Alpha=MaterialFunctionCall() {materialFunction=&/MF_CP_Device_HologramDancer_ColorLerp}.Curve Lerp Alpha
  )
  n99 = Multiply(
   A=Power(Base=Clamp(None=@n91), Exp=ScalarParameter() {defaultValue=4, group=Specular, parameterName=Spec Highlight Power}) {constExponent=20}
   B=@n98
  )
  n100 = ScalarParameter() {defaultValue=2, group=Specular, parameterName=Spec Intense}
  n101 = MaterialFunctionCall(
   In=Fresnel(ExponentIn=ScalarParameter() {defaultValue=1.25, group=Fresnel, parameterName=General FresnelExp}, Normal=VertexNormalWS())
   Contrast=ScalarParameter() {defaultValue=0.5, group=Fresnel, parameterName=General FresnelContrast}
  ) {materialFunction=&CheapContrast}
  n102 = ScalarParameter() {defaultValue=4, group=Fresnel, parameterName=General FresnelIntensity}
  n103 = Add(A=Add(A=Multiply(A=@n88, B=@n89), B=Multiply(A=@n99, B=@n100)), B=Multiply(A=Multiply(A=@n101.Result, B=@n98), B=@n102))
  n104 = Multiply(A=@n98, B=ScalarParameter() {defaultValue=1.5, parameterName=EmissiveMult}) {constB=2}
  NamedRerouteDeclaration(None=StaticSwitchParameter(
    True=VectorParameter() {group=Silhouette, parameterName=Silhouette Color, sortPriority=1, defaultValue=<rgba 0 0 0 1>}.RGB
    False=Add(A=@n103, B=@n104)
   ) {desc=Silhouette Switch, group=Silhouette, parameterName=IsSilhouette, sortPriority=0}) {name=Emissive, nodeColor=<rgba 1 0.8203 0 1>}
  n105 = OneMinus(None=MaterialFunctionCall(UV Channel=@n95.R) {materialFunction=&LinearGradient}.VGradient)
  n106 = ScalarParameter() {defaultValue=2, group=Silhouette, parameterName=Silhouette Mask Power}
  n107 = Multiply(
   A=@n98
   B=MaterialFunctionCall(In=Power(Base=@n105, Exp=@n106), Contrast=ScalarParameter() {defaultValue=0.2, group=Silhouette, parameterName=Silhouette Mask Contrast}) {materialFunction=&CheapContrast}.Result
  )
  Reroute(None=Multiply(A=Multiply(A=@n101.Result, B=@n107), B=@n102))
  n108 = WorldPosition()
  n109 = CameraPositionWS()
  n110 = ActorPositionWS()
  n111 = ScalarParameter() {defaultValue=20000, parameterName=ScaleDistance}
  n112 = Saturate(None=Divide(A=Distance(A=@n109, B=@n110), B=@n111) {constB=1000})
  Reroute(None=Divide(A=ComponentMask(None=@n108.XYZ) {b=true}, B=LinearInterpolate(B=@n111, Alpha=@n112) {constA=50, constB=100}) {constB=20})
  n113 = ComponentMask(None=Subtract(A=WorldPosition().XYZ, B=ObjectPositionWS())) {b=true}
  n114 = ObjectRadius()
  n115 = Saturate(None=ConstantBiasScale(None=Divide(A=@n113, B=@n114)))
  n116 = OneMinus(None=@n115)
  n117 = ScalarParameter() {group=Color, parameterName=FlipGradient, sortPriority=80}
  n118 = Add(
   A=LinearInterpolate(A=@n71, B=LinearInterpolate(A=@n115, B=@n116, Alpha=@n117), Alpha=@n72)
   B=ScalarParameter() {group=Color, parameterName=Fill, sortPriority=80}
  )
  Reroute(None=Multiply(A=ScalarParameter() {defaultValue=1, parameterName=TurnOn}, B=Clamp(None=@n118)))
  n119 = Sine(None=Time()) {period=4}
  n120 = VectorParameter() {parameterName=BobDistance, defaultValue=<rgba 0 0 160 0>}
  Reroute(None=Add(A=@n51, B=Multiply(A=@n119, B=@n120.RGB)))
  ScalarParameter() {defaultValue=700, group=Color, parameterName=GradienRadius, sortPriority=81}
  ScalarParameter() {defaultValue=150, group=Hologram Scanlines, parameterName=ScanlineTile}
  ScalarParameter() {defaultValue=0.6, group=Hologram Scanlines, parameterName=MinScanlineOpacity}
  ScalarParameter() {defaultValue=0.7, group=Hologram Scanlines, parameterName=ScanlineColorMinInfluence}
  ScalarParameter() {defaultValue=-0.02, group=Hologram Scanlines, parameterName=ScanlinePanSpeed}
  SetMaterialAttributes!(
   MaterialAttributes
   Emissive Color=NamedRerouteUsage()
   Opacity=NamedRerouteUsage()
   Refraction=NamedRerouteUsage()
   World Position Offset=NamedRerouteUsage()
  ) {attributeSetTypes [
    #1
    #2
    #3
    #4
   ]}
  n121 = MaterialFunctionCall(Vector 2=ObjectBounds()) {materialFunction=/Engine/Functions/Engine_MaterialFunctions02/Utility/VectorLength}
  n122 = Distance(A=ActorPositionWS(), B=CameraPositionWS())
  StaticSwitchParameter(
   True=MaterialFunctionCall(UVs=@n35, Texture Scale=Divide(A=@n121.V2 Length, B=@n122)) {materialFunction=&ScaleUVsByCenter}.UVs
   False=@n35
  ) {defaultValue=true, group=Hologram Scanlines, parameterName=ScaleByDistance, sortPriority=1}
 ]
}
