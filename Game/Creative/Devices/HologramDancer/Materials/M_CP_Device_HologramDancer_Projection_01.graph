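# Unlit, translucent, two-sided material for the Creative HologramDancer device's projection.
# Animated random-noise sine lines, tiled by object scale through a vertex interpolator, blend with fill by fresnel and face side, plus base gradient, intersection glow and optional top line.
# Curve-atlas gradients color it, either hologram or two lerped dancer curves, with silhouette override.
# Opacity adds vertical falloff, strobe and near-camera fade.
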Material /Game/Creative/Devices/HologramDancer/Materials/M_CP_Device_HologramDancer_Projection_01
graph {
 properties {blendMode=BLEND_Translucent, pixelDepthOffsetMode=PDOM_Legacy, shadingModel=MSM_Unlit, twoSided=true, usedShadingModels=Unlit}
 nodes [
  n1 = MaterialFunctionCall() {materialFunction=/Engine/Functions/Engine_MaterialFunctions02/WorldPositionOffset/ObjectScale}
  n2 = Divide(A=Divide(A=Add(A=@n1.Scale X, B=@n1.Scale X)), B=ScalarParameter() {defaultValue=30, desc=.75, parameterName=GridTileU})
  n3 = Divide(A=@n1.Scale Z, B=ScalarParameter() {defaultValue=100, parameterName=GridTileV})
  n4 = VertexInterpolator(VS=Ceil(None=AppendVector(A=@n2, B=@n3)))
  n5 = ScalarParameter() {defaultValue=500, group=Lines, parameterName=LineTiling, sortPriority=11}
  n6 = Constant() {r=1}
  n7 = Multiply(A=TextureCoordinate(), B=Multiply(A=@n4.PS, B=AppendVector(A=@n5, B=@n6)) {constB=4})
  n8 = Time()
  n9 = ScalarParameter() {defaultValue=0.57, group=Lines, parameterName=LineSpeed, sortPriority=12}
  n10 = Sine(None=Add(
    A=MaterialFunctionCall!(PrimaryTextureCoordinates=@n7, 2ndOrderTextureCoordinates, RandomSeed, NoiseDensityBias, Bypass Noise Influence Crosstalk) {materialFunction=/Game/Effects/Fort_Effects/Materials/_Functions/MF_RandomUVSample}.NoiseWithRange
    B=Multiply(A=@n8, B=@n9)
   ))
  ConstantBiasScale(None=@n10)
  n11 = ScalarParameter() {defaultValue=200, group=Lines, parameterName=LineMask Tiling, sortPriority=20}
  n12 = Constant() {r=1}
  n13 = Multiply(A=TextureCoordinate(), B=Multiply(A=@n4.PS, B=AppendVector(A=@n11, B=@n12)) {constB=4})
  n14 = Constant() {r=64}
  n15 = Time()
  n16 = ScalarParameter() {defaultValue=0.13, group=Lines, parameterName=LineMask Speed, sortPriority=21}
  n17 = Sine(None=Add(
    A=MaterialFunctionCall(PrimaryTextureCoordinates=@n13, RandomSeed=@n14) {materialFunction=&MF_RandomUVSample}.NoiseWithRange
    B=Multiply(A=@n15, B=@n16)
   ))
  n18 = Constant() {r=20}
  n19 = Multiply(
   A=Multiply(A=@n10, B=Power(Base=@n17, Exp=@n18))
   B=ScalarParameter() {defaultValue=2, group=Lines, parameterName=Line Intensity, sortPriority=10}
  )
  LinearInterpolate(A=@n19, Alpha=ScalarParameter() {group=Lines, parameterName=TEST})
  n20 = ScalarParameter() {defaultValue=1, group=Color, parameterName=Emissive Multiplier, sortPriority=52}
  n21 = MaterialFunctionCall(UV Channel=TextureCoordinate() {coordinateIndex=1}) {desc=Used as Gradient Map Coordinates, materialFunction=&LinearGradient}
  n22 = ScalarParameter() {defaultValue=0.4, group=Color, parameterName=Curve Power}
  n23 = CurveAtlasRowParameter(CurveTime=Power(Base=@n21.VGradient, Exp=@n22)) {
   atlas=&AthenaCurveAtlas01
   curve=/Game/Creative/Devices/Generic/LinearColorCurves/Curve_CP_Device_Hologram_01
   defaultValue=282
   desc=Curve Atlas
   group=Color
   parameterName=CurveAtlasAthena
  }
  n24 = Power(Base=OneMinus(None=@n21.VGradient), Exp=@n22)
  n25 = CurveAtlasRowParameter(CurveTime=@n24) {
   atlas=/Game/Creative/Devices/HologramDancer/LinearColorCurves/CurveAtlas_CP_Device_HologramDancer_01
   curve=/Game/Creative/Devices/HologramDancer/LinearColorCurves/Curve_CP_Device_HologramDancer_01
   desc=Curve Atlas
   group=Color
   parameterName=Curve01
  }
  n26 = CurveAtlasRowParameter(CurveTime=@n24) {
   atlas=&CurveAtlas_CP_Device_HologramDancer_01
   curve=/Game/Creative/Devices/HologramDancer/LinearColorCurves/Curve_CP_Device_HologramDancer_02
   defaultValue=1
   desc=Curve Atlas
   group=Color
   parameterName=Curve02
  }
  n27 = StaticSwitchParameter(
   True=@n23
   False=LinearInterpolate(
    A=@n25
    B=@n26
    Alpha=MaterialFunctionCall() {materialFunction=&/MF_CP_Device_HologramDancer_ColorLerp}.Curve Lerp Alpha
   )
  ) {parameterName=Is Hologram}
  n28 = TextureCoordinate()
  n29 = Multiply(
   A=ScalarParameter() {defaultValue=0.5, group=Base Gradient, parameterName=BaseGradientIntensity, sortPriority=10}
   B=SphereMask(
    A=ComponentMask(None=@n28) {g=true}
    B=ScalarParameter() {defaultValue=1, group=Base Gradient, parameterName=BaseGradientCenter}
    Radius=ScalarParameter() {defaultValue=0.15, group=Base Gradient, parameterName=BaseGradientRadius}
    Hardness=ScalarParameter() {group=Base Gradient, parameterName=BaseGradientHardness}
   )
  )
  n30 = TextureCoordinate()
  n31 = Power(
   Base=SphereMask(A=ComponentMask(None=@n30) {g=true}, B=Constant() {r=0.01}, Radius=Constant() {r=0.008}, Hardness=Constant())
   Exp=ScalarParameter() {defaultValue=5, parameterName=TopLineExponent}
  )
  n32 = Power(
   Base=MaterialFunctionCall!(
    Distance Bias=ScalarParameter() {group=Intersection Glow, parameterName=Distance Bias}
    Distance Divisor=ScalarParameter() {defaultValue=30, group=Intersection Glow, parameterName=Distance Divisor}
    Scene Depth
    Pixel Depth
   ) {materialFunction=/Game/Effects/Fort_Effects/Materials/_Functions/MF_DistanceBehindTranslucency_01}.Inverted
   Exp=ScalarParameter() {defaultValue=7, group=Intersection Glow, parameterName=Intersection Exponent}
  )
  n33 = TextureCoordinate()
  n34 = Multiply(
   A=ScalarParameter() {defaultValue=0.1, group=Fill, parameterName=Fill Opacity, sortPriority=10}
   B=SphereMask(
    A=ComponentMask(None=@n33) {g=true}
    B=ScalarParameter() {defaultValue=1, group=Fill, parameterName=Fill VerticalFalloffGradientCenter}
    Radius=ScalarParameter() {defaultValue=0.75, group=Fill, parameterName=Fill VerticalFalloffGradientRadius}
    Hardness=ScalarParameter() {group=Fill, parameterName=Fill VerticalFalloffGradientHardness}
   )
  )
  n35 = ScalarParameter() {defaultValue=0.5, group=Opacity, parameterName=BackFaceOpacity, sortPriority=12}
  n36 = ScalarParameter() {defaultValue=1, group=Opacity, parameterName=FrontFaceOpacity, sortPriority=11}
  n37 = Saturate(None=TwoSidedSign())
  n38 = LinearInterpolate(A=@n35, B=@n36, Alpha=@n37)
  n39 = LinearInterpolate(A=@n35, B=@n36, Alpha=@n37)
  n40 = MaterialFunctionCall(
   Brightness=ScalarParameter() {defaultValue=1.25, group=Opacity, parameterName=FresnelBrightness}
   Exponent=ScalarParameter() {defaultValue=3, group=Opacity, parameterName=FresnelExponent}
  ) {materialFunction=&MF_Fresnel_FX}
  n41 = Add(
   A=StaticSwitchParameter(True=Add(A=@n31, B=@n32), False=@n32) {parameterName=UseTopLine}
   B=LinearInterpolate(A=Multiply(A=@n34, B=@n38), B=Multiply(A=@n19, B=@n39), Alpha=OneMinus(None=@n40.Fresnel Double Sided))
  )
  n42 = Multiply(A=@n29, B=Multiply(A=@n39, B=@n10)) {desc=Noisy Lines for Base}
  n43 = Add(A=Add(A=@n41, B=@n29), B=@n42)
  n44 = Constant() {r=1}
  LinearInterpolate(
   A=Multiply(A=Multiply(A=@n20, B=@n27), B=Add(A=@n43, B=@n44))
   B=VectorParameter() {group=Silhouette, parameterName=Silhouette Color, sortPriority=2, defaultValue=<rgba 0 0 0 1>}.RGB
   Alpha=ScalarParameter() {group=Silhouette, parameterName=Silhouette Color Override, sortPriority=1}
  )
  n45 = Add(A=@n42, B=Add(A=@n29, B=@n41))
  n46 = SphereMask(
   A=ComponentMask(None=TextureCoordinate()) {g=true}
   B=ScalarParameter() {defaultValue=1, group=Opacity, parameterName=VerticalFalloffGradientCenter}
   Radius=ScalarParameter() {defaultValue=1, group=Opacity, parameterName=VerticalFalloffGradientRadius}
   Hardness=ScalarParameter() {defaultValue=0.6, group=Opacity, parameterName=VerticalFalloffGradientHardness}
  )
  n47 = ScalarParameter() {defaultValue=100, group=Opacity, parameterName=NearCameraFadeDistance, sortPriority=50}
  Multiply(
   A=MaterialFunctionCall() {desc=Strobe, materialFunction=&/MF_CP_Device_HologramDancer_Strobe}.Strobe Opacity
   B=Multiply(
    A=Multiply(A=@n45, B=@n46)
    B=MaterialFunctionCall(Distance=@n47) {desc=Near Camera Fade, materialFunction=&Near_Camera_Fade}.Result
   )
  )
  n48 = ComponentMask(None=WorldPosition().XYZ) {b=true}
  n49 = Distance(A=CameraPositionWS(), B=ActorPositionWS())
  n50 = ScalarParameter() {defaultValue=20000, parameterName=ScaleDistance}
  n51 = LinearInterpolate(B=@n50, Alpha=Saturate(None=Divide(A=@n49, B=@n50) {constB=1000})) {constA=50, constB=100}
  n52 = Time()
  n53 = ScalarParameter() {defaultValue=-0.02, group=Screen Image, parameterName=ScanlinePanSpeed}
  n54 = Multiply(
   A=ScalarParameter() {defaultValue=60, group=Screen Image, parameterName=ScanlineTile}
   B=Add(A=Divide(A=@n48, B=@n51) {constB=20}, B=Multiply(A=@n52, B=@n53))
  ) {constA=1}
  n55 = LinearInterpolate(B=ConstantBiasScale(None=Sine(None=@n54)), Alpha=ScalarParameter() {defaultValue=0.6, group=Screen Image, parameterName=ScanlineIntensity}) {constA=1}
  n56 = Multiply(B=ScalarParameter() {defaultValue=1, group=Color, parameterName=Emissive Multiplier, sortPriority=52})
  Reroute(None=LinearInterpolate(A=Constant(), B=Multiply(A=@n55, B=@n56)) {constAlpha=1})
 ]
}
